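# Ambient sound cue for an arcade cabinet prop, mixing a looping music bed with intermittent one-shot stingers.
# A looping branch waits a random few seconds, then randomly plays a Dance or modulated Victory clip.
# A quality-level node keeps that stinger branch on the first two quality slots only, leaving the others just the music.
# Shared attenuation localizes it, and a concurrency override allows one instance, blocking new ones.
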SoundCue /Game/Athena/Sounds/Props/ArcadeCabinet/ArcadeCabinet_MX_Loop_Cue
root=Mixer_0
nodes [
 {id=Delay_0, children [Random_0], properties {delayMax=6, delayMin=3}}
 {id=Looping_0, children [Delay_0]}
 {id=Mixer_0, children [WavePlayer_0, QualityLevel_0], properties {inputVolume [1,1]}}
 {id=Modulator_0, children [WavePlayer_2], properties {volumeMax=1, volumeMin=1}}
 {id=QualityLevel_0, children [Looping_0,Looping_0,null,null]}
 {id=Random_0, children [WavePlayer_1,Modulator_0], properties {weights [1,1]}}
 {id=WavePlayer_0, properties {bLooping=true, soundWaveAssetPtr=&/ArcadeCabinet_MX_Loop}}
 {id=WavePlayer_1, properties {soundWaveAssetPtr=&/ArcadeCabinet_Dance_01}}
 {id=WavePlayer_2, properties {soundWaveAssetPtr=&/ArcadeCabinet_Victory_01}}
]
properties {
 attenuationSettings=&/ArcadeCabinet_Att
 bOverrideConcurrency=true
 maxDistance=1000
 soundClassObject=&Fort_Ambient_Props_Athena
 volumeMultiplier=0.501
 concurrencyOverrides {maxCount=1, resolutionRule=PreventNew}
}
